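# Weapon AnimBP for the vampire Stake Launcher that animates firing, finger idles and visible loaded stakes.
# The event graph polls shotgun ammo, reload and pawn targeting state, inferring firing from ammo drops.
# The fire state machine feeds a slot, with random and ADS idles layered onto finger bones, faded out during reload.
# Stake bones scale to zero from an ammo offset that animation curves help time.
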AnimBlueprint /VampireStakeGameplay/Items/StakeLauncher/Animation/StakeLauncher_Weapon_AnimBP
target_skeleton=/VampireStakeGameplay/Items/StakeLauncher/Meshes/Stake_Launcher_Skeleton
graphs [
 AnimGraph {nodes [
   AnimGraphNode_StateMachine_0 = StakeLauncher_Fire_SM() {is=title AnimGraphNode_StateMachine}
   AnimGraphNode_StateMachine_1 = RandomIdle_SM() {is=title AnimGraphNode_StateMachine}
   n1 = Layered blend per bone(
    BasePose=Slot 'DefaultSlot'(Source=@AnimGraphNode_StateMachine_0)
    BlendPoses_0=@AnimGraphNode_StateMachine_1
    BlendWeights_0=Get IdleAlpha() -> (IdleAlpha:double)
   ) {layerSetup [{branchFilters [
       {boneName=finger_a_01_l}
       {boneName=finger_a_01_r}
       {boneName=finger_b_01_l}
       {boneName=finger_b_01_r}
       {boneName=finger_c_01_l}
       {boneName=finger_c_01_r}
       {boneName=finger_d_01_l}
       {boneName=finger_d_01_r}
       {boneName=finger_e_01_l}
       {boneName=finger_e_01_r}
       {boneName=finger_f_01_l}
       {boneName=finger_f_01_r}
       {boneName=finger_g_01_l}
       {boneName=finger_g_01_r}
      ]}]}
   n2 = Get AmmoCountOffset() -> (AmmoCountOffset:int=0)
   n3 = "Transform (Modify) Bone - Bone: stake_04"(
    ComponentPose=Local To Component(LocalPose=@n1)
    Alpha="To Float (Boolean)"(InBool="integer < integer"(A=@n2, B=4))
    Translation="0,0,0"
    Rotation="0,0,0"
    Scale="0,0,0"
   ) {scaleMode=BMM_Replace, scaleSpace=BCS_BoneSpace, alphaBoolBlend {blendOption=Linear}, boneToModify {boneName=stake_04}, scale=<0 0 0>}
   n4 = "Transform (Modify) Bone - Bone: stake_03"(ComponentPose=@n3, Alpha="To Float (Boolean)"(InBool="integer < integer"(A=@n2, B=3)), Translation="0,0,0", Rotation="0,0,0", Scale="0,0,0") {scaleMode=BMM_Replace, scaleSpace=BCS_BoneSpace, alphaBoolBlend {blendOption=Linear}, boneToModify {boneName=stake_03}, scale=<0 0 0>}
   n5 = "Transform (Modify) Bone - Bone: stake_02"(ComponentPose=@n4, Alpha="To Float (Boolean)"(InBool="integer < integer"(A=@n2, B=2)), Translation="0,0,0", Rotation="0,0,0", Scale="0,0,0") {scaleMode=BMM_Replace, scaleSpace=BCS_BoneSpace, alphaBoolBlend {blendOption=Linear}, boneToModify {boneName=stake_02}, scale=<0 0 0>}
   n6 = "Transform (Modify) Bone - Bone: stake_01"(ComponentPose=@n5, Alpha="To Float (Boolean)"(InBool="integer < integer"(A=@n2, B=1)), Translation="0,0,0", Rotation="0,0,0", Scale="0,0,0") {scaleMode=BMM_Replace, scaleSpace=BCS_BoneSpace, alphaBoolBlend {blendOption=Linear}, boneToModify {boneName=stake_01}, scale=<0 0 0>}
   Output Pose(Result=Component To Local(ComponentPose=@n6)) {layerGroup=None}
  ]}
 StakeLauncher_Fire_SM {
  entry=Idle
  path=AnimGraph.AnimGraphNode_StateMachine_0.StakeLauncher_Fire_SM
  ids=id
  nodes [
   Reload() {is=title AnimStateAliasNode, bGlobalAlias=true}
   AnimStateNode_10 = Idle()
   AnimStateNode_11 = Shoot() {is=title AnimStateNode}
   AnimStateTransitionNode_1 = AnimStateTransitionNode() {crossfadeDuration=0}
   AnimStateTransitionNode_2 = AnimStateTransitionNode() {bAutomaticRuleBasedOnSequencePlayerInState=true, crossfadeDuration=0}
   AnimStateTransitionNode_3 = AnimStateTransitionNode() {crossfadeDuration=0}
  ]
  links [
   {
    dst=Shoot
    rule=AnimGraph.AnimGraphNode_StateMachine_0.StakeLauncher_Fire_SM.AnimStateTransitionNode_1.Transition
    src=Idle
    via=AnimStateTransitionNode_1
   }
   {
    dst=Idle
    rule=AnimGraph.AnimGraphNode_StateMachine_0.StakeLauncher_Fire_SM.AnimStateTransitionNode_3.Transition
    src=Reload
    via=AnimStateTransitionNode_3
   }
   {
    dst=Idle
    rule=AnimGraph.AnimGraphNode_StateMachine_0.StakeLauncher_Fire_SM.AnimStateTransitionNode_2.Transition
    src=Shoot
    via=AnimStateTransitionNode_2
   }
  ]
 }
 Idle {
  path=AnimGraph.AnimGraphNode_StateMachine_0.StakeLauncher_Fire_SM.AnimStateNode_10.Idle
  nodes [Output Animation Pose(Result=Sequence Player 'Launcher_StakeLauncher_BasePose_WEP'() {bLoopAnimation=false, sequence=&/Launcher_StakeLauncher_BasePose_WEP})]
 }
 Shoot {
  path=AnimGraph.AnimGraphNode_StateMachine_0.StakeLauncher_Fire_SM.AnimStateNode_11.Shoot
  nodes [Output Animation Pose(Result=Sequence Player 'Launcher_StakeLauncher_Fire_WEP'() {bLoopAnimation=false, sequence=&/Launcher_StakeLauncher_Fire_WEP})]
 }
 Transition {
  path=AnimGraph.AnimGraphNode_StateMachine_0.StakeLauncher_Fire_SM.AnimStateTransitionNode_1.Transition
  nodes [Result(bCanEnterTransition=Get IsFiring() -> (IsFiring:bool))]
 }
 Transition {path=AnimGraph.AnimGraphNode_StateMachine_0.StakeLauncher_Fire_SM.AnimStateTransitionNode_2.Transition, nodes [Result()]}
 Transition {
  path=AnimGraph.AnimGraphNode_StateMachine_0.StakeLauncher_Fire_SM.AnimStateTransitionNode_3.Transition
  nodes [Result(bCanEnterTransition=Get IsReloading() -> (IsReloading:bool))]
 }
 RandomIdle_SM {
  entry=Random Idle
  path=AnimGraph.AnimGraphNode_StateMachine_1.RandomIdle_SM
  ids=id
  nodes [
   AnimStateNode_0 = Random Idle() {is=title AnimStateNode}
   AnimStateNode_1 = ADS Idle() {is=title AnimStateNode}
   AnimStateTransitionNode_0 = AnimStateTransitionNode() {crossfadeDuration=0.1}
   AnimStateTransitionNode_1 = AnimStateTransitionNode() {bAutomaticRuleBasedOnSequencePlayerInState=true, crossfadeDuration=0.1}
  ]
  links [
   {
    dst=Random Idle
    rule=AnimGraph.AnimGraphNode_StateMachine_1.RandomIdle_SM.AnimStateTransitionNode_1.Transition
    src=ADS Idle
    via=AnimStateTransitionNode_1
   }
   {
    dst=ADS Idle
    rule=AnimGraph.AnimGraphNode_StateMachine_1.RandomIdle_SM.AnimStateTransitionNode_0.Transition
    src=Random Idle
    via=AnimStateTransitionNode_0
   }
  ]
 }
 Random Idle {
  path=AnimGraph.AnimGraphNode_StateMachine_1.RandomIdle_SM.AnimStateNode_0.Random Idle
  nodes [
   n1 = Random Sequence Player!() -> (Pose:Pose Link Structure) {
    is=title AnimGraphNode_RandomPlayer
    bShuffleMode=true
    entries [
     {
      chanceToPlay=1
      maxLoopCount=2
      maxPlayRate=1
      minLoopCount=1
      minPlayRate=1
      sequence=&/Launcher_StakeLauncher_Idle_03_Motionless_WEP
      blendIn {blendOption=Linear, blendTime=0, customCurve=null}
     }
     {
      chanceToPlay=1
      maxLoopCount=1
      maxPlayRate=1
      minLoopCount=1
      minPlayRate=1
      sequence=&/Launcher_StakeLauncher_Idle_01_ConsecutiveTaps_WEP
      blendIn {blendOption=Linear, blendTime=0, customCurve=null}
     }
     {
      chanceToPlay=1
      maxLoopCount=1
      maxPlayRate=1
      minLoopCount=1
      minPlayRate=1
      sequence=&/Launcher_StakeLauncher_Idle_02_Scratch_WEP
      blendIn {blendOption=Linear, blendTime=0, customCurve=null}
     }
     {
      chanceToPlay=1
      maxLoopCount=2
      maxPlayRate=1
      minLoopCount=2
      minPlayRate=1
      sequence=&Launcher_StakeLauncher_Idle_03_Motionless_WEP
      blendIn {blendOption=Linear, blendTime=0, customCurve=null}
     }
     {
      chanceToPlay=1
      maxLoopCount=2
      maxPlayRate=1
      minLoopCount=1
      minPlayRate=1
      sequence=&Launcher_StakeLauncher_Idle_03_Motionless_WEP
      blendIn {blendOption=Linear, blendTime=0, customCurve=null}
     }
     {
      chanceToPlay=1
      maxLoopCount=2
      maxPlayRate=1
      minLoopCount=1
      minPlayRate=1
      sequence=&Launcher_StakeLauncher_Idle_03_Motionless_WEP
      blendIn {blendOption=Linear, blendTime=0, customCurve=null}
     }
     {
      chanceToPlay=1
      maxLoopCount=2
      maxPlayRate=1
      minLoopCount=1
      minPlayRate=1
      sequence=&Launcher_StakeLauncher_Idle_03_Motionless_WEP
      blendIn {blendOption=Linear, blendTime=0, customCurve=null}
     }
     {
      chanceToPlay=1
      maxLoopCount=2
      maxPlayRate=1
      minLoopCount=1
      minPlayRate=1
      sequence=&Launcher_StakeLauncher_Idle_03_Motionless_WEP
      blendIn {blendOption=Linear, blendTime=0, customCurve=null}
     }
    ]
   }
   Output Animation Pose(Result=@n1)
  ]
 }
 ADS Idle {
  path=AnimGraph.AnimGraphNode_StateMachine_1.RandomIdle_SM.AnimStateNode_1.ADS Idle
  nodes [
   n1 = Sequence Player 'Launcher_StakeLauncher_Idle_01_ConsecutiveTaps_WEP'() {bLoopAnimation=false, sequence=&Launcher_StakeLauncher_Idle_01_ConsecutiveTaps_WEP}
   Output Animation Pose(Result=@n1)
  ]
 }
 Transition {
  path=AnimGraph.AnimGraphNode_StateMachine_1.RandomIdle_SM.AnimStateTransitionNode_0.Transition
  nodes [
   n1 = Get IsADSIdleDone() -> (IsADSIdleDone:bool)
   Result(bCanEnterTransition=AND Boolean(A=Get IsADSIdle() -> (IsADSIdle:bool), B=NOT Boolean(A=@n1)))
  ]
 }
 Transition {path=AnimGraph.AnimGraphNode_StateMachine_1.RandomIdle_SM.AnimStateTransitionNode_1.Transition, nodes [Result()]}
 EventGraph {nodes [
   PrintString!(
    execute
    InString="To String (Integer)"(InInt=Get AmmoCount() -> (AmmoCount:int=0))
    bPrintToScreen
    bPrintToLog
    TextColor=(R=0,G=1,B=0.25,A=1)
    Duration=0.1
    Key
   ) -> (then)
   > PrintString(InString="To String (Integer)"(InInt=Get AmmoCountOffset() -> (AmmoCountOffset=0)), TextColor=(R=0.9905,G=1,B=0,A=1), Duration=0.1)
   > PrintString(InString="To String (Boolean)"(InBool=Get IsReloading()), TextColor=(R=1,G=0,B=0.0331,A=1), Duration=0.1)
   n1 = Event BlueprintUpdateAnimation()
   > Set Delta Time X(Delta Time X=@n1.DeltaTimeX)
   > n2 = Sequence!(execute) -> (then_0, then_1, then_2:Exec, then_3:Exec, then_4:Exec, then_5:Exec, then_6:Exec)
   n3 = TryGetFortPawn()
   > Is Valid(InputObject=@n3)
   > n4 = Cast To FortPlayerPawn(Object=@n3)
   > n5 = Set As Fort Player Pawn(As Fort Player Pawn:Fort Player Pawn Object Reference=@n4) -> (Output_Get:Fort Player Pawn Object Reference)
   > Set IsADS(IsADS:bool=HasMatchingGameplayTag!(self:Gameplay Tag Asset Interface Interface=@n5, TagToCheck:Gameplay Tag Structure=(TagName="Gameplay.Status.IsTargeting")) -> (ReturnValue:bool)) -> (Output_Get:bool)
   n6 = Get As B Vampire Stake Shotgun() -> (As B Vampire Stake Shotgun:B Vampire Stake Shotgun Object Reference)
   Is Valid(exec=@n2.then_1, InputObject=@n6)
   > Set AmmoCount(AmmoCount:int=GetMagazineAmmoCount!(self:Fort Weapon Object Reference=@n6) -> (ReturnValue:int=0)) -> (Output_Get:int=0)
   > n7 = Set IsReloading(IsReloading:bool=IsReloading!(self:Fort Weapon Object Reference=@n6) -> (ReturnValue:bool)) -> (Output_Get:bool)
   n8 = NOT Boolean(A=@n7)
   > Set IdleAlpha(IdleAlpha:double=FInterpTo(Current=Get IdleAlpha(), Target="To Float (Boolean)"(InBool=@n8), DeltaTime=Get Delta Time X(), InterpSpeed=6)) -> (Output_Get:double)
   Is Valid(exec=@n2.then_2, InputObject=Get As B Vampire Stake Shotgun())
   > Is Valid(InputObject=Get As Fort Player Pawn() -> (As Fort Player Pawn:Fort Player Pawn Object Reference))
   > Branch(Condition="Not Equal (Integer)"(A=Get AmmoCount() -> (AmmoCount=0), B=Get AmmoCountOffset() -> (AmmoCountOffset=0)))
   > Do Once()
   > Set AmmoCountOffset(AmmoCountOffset:int=Get AmmoCount() -> (AmmoCount=0)) -> (Output_Get:int=0)
   Is Valid(exec=@n2.then_3, InputObject=Get As B Vampire Stake Shotgun())
   n9 = Get IsReloading()
   n10 = Get AmmoCountOffset() -> (AmmoCountOffset=0)
   n11 = Get AmmoCount() -> (AmmoCount=0)
   > n12 = Branch(Condition=AND Boolean(A="integer < integer"(A=@n11, B=@n10), B=NOT Boolean(A=@n9)))
   > Set AmmoCountOffset(AmmoCountOffset=int + int(A=Get AmmoCount() -> (AmmoCount=0))) -> (Output_Get=0)
   > Set IsFiring(IsFiring:bool=true) -> (Output_Get:bool)
   > Branch(Condition="float > float"(A=GetCurveValue(CurveName=UserCurve01), B=0.001))
   > Set IsFiring(IsFiring=false)
   > Set AmmoCountOffset(AmmoCountOffset=Get AmmoCount() -> (AmmoCount=0)) -> (Output_Get=0)
   Set IsFiring(execute=@n12.else, IsFiring=false)
   n13 = Branch(execute=@n2.then_4, Condition=Get IsReloading())
   > Do Once(Reset=@n13.else)
   > Set AmmoCountOffset(AmmoCountOffset=Get AmmoCount() -> (AmmoCount=0)) -> (Output_Get=0)
   n14 = GetCurveValue(CurveName=UserCurve01)
   n15 = Branch(execute=@n2.then_5, Condition=AND Boolean(A=Get IsReloading(), B="float > float"(A=@n14, B=0.001)))
   > Do Once(Reset=@n15.else)
   n16 = Get AmmoCountOffset() -> (AmmoCountOffset=0)
   > Set AmmoCountOffset(AmmoCountOffset="Clamp (Integer)"(Value=int + int(A=@n16), Min=0, Max=4) -> (ReturnValue=0)) -> (Output_Get=0)
   n17 = Branch(execute=@n2.then_6, Condition=Get IsADS() -> (IsADS:bool))
   > Delay(Duration=0.8)
   n18 = GetCurveValue(CurveName=UserCurve02)
   > Branch(Condition=AND Boolean(A=RandomBool!() -> (ReturnValue:bool), B="float > float"(A=@n18, B=0.5)))
   > Set IsADSIdle(IsADSIdle:bool=true) -> (Output_Get:bool)
   > Delay(Duration=0.1)
   > Set IsADSIdleDone(IsADSIdleDone:bool=true) -> (Output_Get:bool)
   Set IsADSIdle(execute=@n17.else, IsADSIdle=false)
   > Set IsADSIdleDone(IsADSIdleDone=false)
   Event BlueprintInitializeAnimation()
   n19 = GetOwningActor()
   > Is Valid(InputObject=@n19)
   > n20 = Cast To B_VampireStake_Shotgun!(execute, Object=@n19) -> (then, CastFailed, AsB Vampire Stake Shotgun:B Vampire Stake Shotgun Object Reference) {is=title K2Node_DynamicCast}
   > Set As B Vampire Stake Shotgun(As B Vampire Stake Shotgun:B Vampire Stake Shotgun Object Reference=@n20) -> (Output_Get:B Vampire Stake Shotgun Object Reference)
   Get AmmoCount() -> (AmmoCount=0)
  ]}
]
variables [
 AmmoCount {type=integer}
 AmmoCountOffset {type=integer}
 As B Vampire Stake Shotgun {type=B_VampireStake_Shotgun_C, default=null}
 As Fort Player Pawn {type=FortPlayerPawn}
 Delta Time X {type=number}
 IdleAlpha {type=number}
 IsADS {type=boolean}
 IsADSIdle {type=boolean}
 IsADSIdleDone {type=boolean}
 IsFiring {type=boolean}
 IsReloading {type=boolean}
]
