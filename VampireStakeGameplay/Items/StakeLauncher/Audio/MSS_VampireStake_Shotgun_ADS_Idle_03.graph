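# MetaSound one-shot mono source that plays a stake launcher idle foley wave (variant 03), named for the shotgun's aim-down-sights state.
# On play, a trigger compare with no wired input either starts the wave player or finishes the source.
# Changes to the IsADS bool run a second compare that either stops the player or finishes the source.
# Wave completion also fires OnFinished, merged through a Trigger Any.
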MetaSoundSource /VampireStakeGameplay/Items/StakeLauncher/Audio/MSS_VampireStake_Shotgun_ADS_Idle_03
inputs [
 IsADS {type=Bool}
 UE.Source.OnPlay {type=Trigger}
]
interfaces [UE.OutputFormat.Mono:1.0, UE.Source.OneShot:1.0, UE.Source:1.1]
outputs [
 "UE.OutputFormat.Mono.Audio:0" {type=Audio}
 UE.Source.OneShot.OnFinished {type=Trigger}
]
pages [Default {
  input_defaults {IsADS=null, UE.Source.OnPlay=null}
  nodes [
   TriggerOnValueChange_0 = UE.Trigger On Value Change:Bool(Value=^IsADS)
   n1 = TriggerCompare.Clamp:Bool(A=UE.Reroute(Value=^IsADS).Value, Compare=@TriggerOnValueChange_0.Trigger) {also {defaults {B="<MetaSoundBuilderResult.SUCCEEDED: 0>"}}}
   n2 = TriggerCompare.Clamp:Bool(Compare=^UE.Source.OnPlay) {also {defaults {B="<MetaSoundBuilderResult.SUCCEEDED: 0>"}}}
   WavePlayer_0 = UE.Wave Player:Mono(Stop=@n1.False, Play=@n2.True) {also {defaults {Wave Asset=/VampireStakeGameplay/Items/StakeLauncher/Audio/SoundWaves/StakeLauncher_Idle_03}}}
   TriggerAnyTriggerAny3_0 = "TriggerAny.Trigger Any (3)"(In 2=@n1.True, In 1=@n2.False, In 0=@WavePlayer_0.On Finished)
  ]
  links [{dst="@UE.Source.OneShot.OnFinished", src=TriggerAnyTriggerAny3_0.Out}, {dst="@UE.OutputFormat.Mono.Audio:0", src=WavePlayer_0.Out Mono}]
 }]
properties {
 assetImportData=&self:AssetImportData
 attenuationSettings=/VampireStakeGameplay/Items/StakeLauncher/Audio/Settings/StakeLauncher_Legs_Close
 maxDistance=70
 sampleRate=48000
 soundClassObject=&Fort_Player_Foley_1P_Athena
 volume=0.3
}
