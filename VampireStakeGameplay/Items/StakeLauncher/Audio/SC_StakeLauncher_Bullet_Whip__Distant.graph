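# Sound cue for the Stake Launcher's distant bullet whip, the sound of a projectile passing at a distance.
# A Random node picks one of seven equally weighted wave variations and feeds a Modulator that varies pitch and volume on each play.
# It uses shared Athena bullet whip attenuation and whiz-by concurrency, routes to the ranged projectile effects sound class, and has virtualization disabled.
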SoundCue /VampireStakeGameplay/Items/StakeLauncher/Audio/SC_StakeLauncher_Bullet_Whip__Distant
root=Modulator_0
nodes [
 {id=Modulator_0, children [Random_1], properties {pitchMax=1.2, volumeMax=1, volumeMin=0.5}}
 {
  id=Random_1
  children [WavePlayer_0, WavePlayer_5, WavePlayer_6, WavePlayer_7, WavePlayer_8, WavePlayer_9, WavePlayer_10]
  properties {bShouldExcludeFromBranchCulling=true, weights [1,1,1,1,1,1,1]}
 }
 {
  id=WavePlayer_0
  properties {soundWaveAssetPtr=/VampireStakeGameplay/Items/StakeLauncher/Audio/SoundWaves/StakeLauncher_BulletWhip_Distant_01}
 }
 {
  id=WavePlayer_10
  properties {soundWaveAssetPtr=/VampireStakeGameplay/Items/StakeLauncher/Audio/SoundWaves/StakeLauncher_BulletWhip_Distant_07}
 }
 {
  id=WavePlayer_5
  properties {soundWaveAssetPtr=/VampireStakeGameplay/Items/StakeLauncher/Audio/SoundWaves/StakeLauncher_BulletWhip_Distant_02}
 }
 {
  id=WavePlayer_6
  properties {soundWaveAssetPtr=/VampireStakeGameplay/Items/StakeLauncher/Audio/SoundWaves/StakeLauncher_BulletWhip_Distant_03}
 }
 {
  id=WavePlayer_7
  properties {soundWaveAssetPtr=/VampireStakeGameplay/Items/StakeLauncher/Audio/SoundWaves/StakeLauncher_BulletWhip_Distant_04}
 }
 {
  id=WavePlayer_8
  properties {soundWaveAssetPtr=/VampireStakeGameplay/Items/StakeLauncher/Audio/SoundWaves/StakeLauncher_BulletWhip_Distant_05}
 }
 {
  id=WavePlayer_9
  properties {soundWaveAssetPtr=/VampireStakeGameplay/Items/StakeLauncher/Audio/SoundWaves/StakeLauncher_BulletWhip_Distant_06}
 }
]
properties {
 attenuationSettings=/Game/Athena/Sounds/Weapons/BulletWhips/bullet_whip_Att.Bullet_Whip_Att
 bBypassVolumeScaleForPriority=true
 maxDistance=2097152
 priority=0.5
 soundClassObject=&Fort_WeaponRanged_Projectile_Effects
 virtualizationMode=Disabled
 volumeMultiplier=0.7
 attenuationOverrides {bAttenuate=false, stereoSpread=0}
 concurrencySet [/Game/Sounds/Concurrency_Rules/Athena_BulletWhizBy_Conc]
}
