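# Sound Cue for the Stake Launcher's first-person firing sound.
# A Random node picks one of eight equally weighted fire waves, and a Modulator at the root adds slight upward pitch variation.
# The Random node is excluded from branch culling.
# First-person projectile attenuation, a ranged-fire sound class and a shared first-person gunfire concurrency set govern playback, and the cue's debug flag is left on.
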SoundCue /VampireStakeGameplay/Items/StakeLauncher/Audio/SC_StakeLauncher_Fire_1P
root=Modulator_0
nodes [
 {id=Modulator_0, children [Random_0], properties {pitchMax=1.02, pitchMin=1}}
 {
  id=Random_0
  children [WavePlayer_0,WavePlayer_1,WavePlayer_2,WavePlayer_3,WavePlayer_4,WavePlayer_5,WavePlayer_6,WavePlayer_7]
  properties {bShouldExcludeFromBranchCulling=true, weights [1,1,1,1,1,1,1,1]}
 }
 {id=WavePlayer_0, properties {soundWaveAssetPtr=/VampireStakeGameplay/Items/StakeLauncher/Audio/SoundWaves/StakeLauncher_Fire_1P_01}}
 {id=WavePlayer_1, properties {soundWaveAssetPtr=/VampireStakeGameplay/Items/StakeLauncher/Audio/SoundWaves/StakeLauncher_Fire_1P_02}}
 {id=WavePlayer_2, properties {soundWaveAssetPtr=/VampireStakeGameplay/Items/StakeLauncher/Audio/SoundWaves/StakeLauncher_Fire_1P_03}}
 {id=WavePlayer_3, properties {soundWaveAssetPtr=/VampireStakeGameplay/Items/StakeLauncher/Audio/SoundWaves/StakeLauncher_Fire_1P_04}}
 {id=WavePlayer_4, properties {soundWaveAssetPtr=/VampireStakeGameplay/Items/StakeLauncher/Audio/SoundWaves/StakeLauncher_Fire_1P_05}}
 {id=WavePlayer_5, properties {soundWaveAssetPtr=/VampireStakeGameplay/Items/StakeLauncher/Audio/SoundWaves/StakeLauncher_Fire_1P_06}}
 {id=WavePlayer_6, properties {soundWaveAssetPtr=/VampireStakeGameplay/Items/StakeLauncher/Audio/SoundWaves/StakeLauncher_Fire_1P_07}}
 {id=WavePlayer_7, properties {soundWaveAssetPtr=/VampireStakeGameplay/Items/StakeLauncher/Audio/SoundWaves/StakeLauncher_Fire_1P_08}}
]
properties {
 attenuationSettings=&Fort_Weapon_Projectile_FirstPerson_Athena_Att
 bDebug=true
 maxDistance=2097152
 soundClassObject=&Fort_WeaponRanged_Fire_1P_PSM
 volumeMultiplier=0.9
 concurrencyOverrides {bLimitToOwner=true, maxCount=21, resolutionRule=StopOldest}
 concurrencySet [&Weapon_Gunfire_FirstPerson_Conc]
}
