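# Sound cue for Stake Launcher impacts on Husks, layering random hard stake hits with random Husk body impacts, each pitch-varied.
# IsEnemy and IsLocal branches choose either a quieter, extra-delayed instigator hit-notification path or a louder player-impact path, each with its own sound class and attenuation.
# Both share a fixed delay and a quality-level switch between full and reduced stake variation pools.
# Rifle bullet impact concurrency is assigned.
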SoundCue /VampireStakeGameplay/Items/StakeLauncher/Audio/SC_StakeLauncher_Husk_Impact
root=Branch_0
nodes [
 {id=Attenuation_0, children [Delay_0], properties {attenuationSettings=/Game/Athena/Sounds/HitNotifications/Fort_HitMarker_Instigate_Att}}
 {id=Attenuation_1, children [Delay_0], properties {attenuationSettings=/Game/Sounds/Attenuation/Fort_Bullet_ImpactOpponent_Lng_Att}}
 {id=Branch_0, children [Modulator_4,Branch_1,Branch_1], properties {boolParameterName=IsEnemy}}
 {id=Branch_1, children [Modulator_5,Modulator_4,Modulator_4], properties {boolParameterName=IsLocal}}
 {id=Delay_0, children [QualityLevel_1], properties {delayMax=0.15, delayMin=0.15}}
 {id=Delay_1, children [SoundClass_0], properties {delayMax=0.09, delayMin=0.08}}
 {id=Mixer_0, children [Modulator_1,Modulator_0], properties {inputVolume [1,1]}}
 {id=Mixer_1, children [Modulator_2,Modulator_3], properties {inputVolume [1,1]}}
 {id=Modulator_0, children [Random_1], properties {pitchMax=1.3, pitchMin=0.9}}
 {id=Modulator_1, children [Random_6], properties {pitchMax=1.2, pitchMin=1}}
 {id=Modulator_2, children [Random_0], properties {pitchMax=1.2, pitchMin=1}}
 {id=Modulator_3, children [Random_2], properties {pitchMax=1.3, pitchMin=0.9}}
 {id=Modulator_4, children [SoundClass_1], properties {pitchMax=1, pitchMin=1, volumeMax=2, volumeMin=2}}
 {id=Modulator_5, children [Delay_1], properties {pitchMax=1, pitchMin=1, volumeMax=0.5, volumeMin=0.5}}
 {id=QualityLevel_1, children [Mixer_1,Mixer_1,Mixer_0,Mixer_0]}
 {
  id=Random_0
  children [WavePlayer_0,WavePlayer_1,WavePlayer_2,WavePlayer_3,WavePlayer_4,WavePlayer_5,WavePlayer_6,WavePlayer_7,WavePlayer_8,WavePlayer_9]
  properties {bShouldExcludeFromBranchCulling=true, weights [1,1,1,1,1,1,1,1,1,1]}
 }
 {
  id=Random_1
  children [WavePlayer_36, WavePlayer_37, WavePlayer_38, WavePlayer_39]
  properties {bShouldExcludeFromBranchCulling=true, weights [1,1,1,1]}
 }
 {id=Random_2, children [WavePlayer_11, WavePlayer_10, WavePlayer_25, WavePlayer_28], properties {weights [1,1,1,1]}}
 {
  id=Random_6
  children [WavePlayer_35, WavePlayer_34, WavePlayer_33, WavePlayer_32, WavePlayer_29]
  properties {bShouldExcludeFromBranchCulling=true, weights [1,1,1,1,1]}
 }
 {
  id=SoundClass_0
  children [Attenuation_0]
  properties {soundClassOverride=&Fort_2DHitNotifications_Instigate_PSM}
 }
 {id=SoundClass_1, children [Attenuation_1], properties {soundClassOverride=/Game/Sounds/FortSoundClassesAndModes/Fort_Impact_Player_PSM}}
 {id=WavePlayer_0, properties {soundWaveAssetPtr=/VampireStakeGameplay/Items/StakeLauncher/Audio/SoundWaves/StakeLauncher_Impact_Hard_01}}
 {id=WavePlayer_1, properties {soundWaveAssetPtr=/VampireStakeGameplay/Items/StakeLauncher/Audio/SoundWaves/StakeLauncher_Impact_Hard_02}}
 {id=WavePlayer_10, properties {soundWaveAssetPtr=/Game/Sounds/Fort_Monster_Husk/SFX/husk_body_impact_mix_02}}
 {id=WavePlayer_11, properties {soundWaveAssetPtr=/Game/Sounds/Fort_Monster_Husk/SFX/husk_body_impact_mix_01}}
 {id=WavePlayer_2, properties {soundWaveAssetPtr=/VampireStakeGameplay/Items/StakeLauncher/Audio/SoundWaves/StakeLauncher_Impact_Hard_03}}
 {id=WavePlayer_25, properties {soundWaveAssetPtr=/Game/Sounds/Fort_Monster_Husk/SFX/husk_body_impact_mix_03}}
 {id=WavePlayer_28, properties {soundWaveAssetPtr=/Game/Sounds/Fort_Monster_Husk/SFX/husk_body_impact_mix_04}}
 {id=WavePlayer_29, properties {soundWaveAssetPtr=/VampireStakeGameplay/Items/StakeLauncher/Audio/SoundWaves/StakeLauncher_Impact_Hard_10}}
 {id=WavePlayer_3, properties {soundWaveAssetPtr=/VampireStakeGameplay/Items/StakeLauncher/Audio/SoundWaves/StakeLauncher_Impact_Hard_04}}
 {id=WavePlayer_32, properties {soundWaveAssetPtr=/VampireStakeGameplay/Items/StakeLauncher/Audio/SoundWaves/StakeLauncher_Impact_Hard_08}}
 {id=WavePlayer_33, properties {soundWaveAssetPtr=/VampireStakeGameplay/Items/StakeLauncher/Audio/SoundWaves/StakeLauncher_Impact_Hard_06}}
 {id=WavePlayer_34, properties {soundWaveAssetPtr=&StakeLauncher_Impact_Hard_04}}
 {id=WavePlayer_35, properties {soundWaveAssetPtr=&StakeLauncher_Impact_Hard_03}}
 {id=WavePlayer_36, properties {soundWaveAssetPtr=&husk_body_impact_mix_01}}
 {id=WavePlayer_37, properties {soundWaveAssetPtr=&husk_body_impact_mix_02}}
 {id=WavePlayer_38, properties {soundWaveAssetPtr=&husk_body_impact_mix_03}}
 {id=WavePlayer_39, properties {soundWaveAssetPtr=&husk_body_impact_mix_04}}
 {id=WavePlayer_4, properties {soundWaveAssetPtr=/VampireStakeGameplay/Items/StakeLauncher/Audio/SoundWaves/StakeLauncher_Impact_Hard_05}}
 {id=WavePlayer_5, properties {soundWaveAssetPtr=&StakeLauncher_Impact_Hard_06}}
 {id=WavePlayer_6, properties {soundWaveAssetPtr=/VampireStakeGameplay/Items/StakeLauncher/Audio/SoundWaves/StakeLauncher_Impact_Hard_07}}
 {id=WavePlayer_7, properties {soundWaveAssetPtr=&StakeLauncher_Impact_Hard_08}}
 {id=WavePlayer_8, properties {soundWaveAssetPtr=/VampireStakeGameplay/Items/StakeLauncher/Audio/SoundWaves/StakeLauncher_Impact_Hard_09}}
 {id=WavePlayer_9, properties {soundWaveAssetPtr=&StakeLauncher_Impact_Hard_10}}
]
properties {
 maxDistance=2097152
 soundClassObject=&Fort_Impacts
 volumeMultiplier=1.5
 attenuationOverrides {falloffDistance=5000, attenuationShapeExtents {x=1000}}
 concurrencyOverrides {maxCount=4, volumeScale=0.8}
 concurrencySet [&RifleBulletImpact_Conc]
}
