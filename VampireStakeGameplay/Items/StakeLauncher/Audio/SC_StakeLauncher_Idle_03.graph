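# Sound cue for variant 03 of the Stake Launcher's idle sound, classed as first-person player foley at reduced volume.
# A Branch on Weapon.IsAimingDownSights feeds a Modulator and plays the wave only when the parameter is true; the false and unset inputs are empty.
# It references a close-range legs attenuation asset and also carries short-range falloff values.
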SoundCue /VampireStakeGameplay/Items/StakeLauncher/Audio/SC_StakeLauncher_Idle_03
root=Modulator_0
nodes [
 {id=Branch_0, children [WavePlayer_0,null,null], properties {boolParameterName=Weapon.IsAimingDownSights}}
 {id=Modulator_0, children [Branch_0]}
 {id=WavePlayer_0, properties {soundWaveAssetPtr=/VampireStakeGameplay/Items/StakeLauncher/Audio/SoundWaves/StakeLauncher_Idle_03}}
]
properties {
 attenuationSettings=/VampireStakeGameplay/Items/StakeLauncher/Audio/Settings/StakeLauncher_Legs_Close
 maxDistance=70
 soundClassObject=&Fort_Player_Foley_1P_Athena
 volumeMultiplier=0.3
 attenuationOverrides {falloffDistance=50, attenuationShapeExtents {x=20}}
}
