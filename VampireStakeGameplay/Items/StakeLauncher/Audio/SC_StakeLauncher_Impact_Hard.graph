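# Sound cue for the Stake Launcher's hard impact, played in the impacts sound class with its own attenuation asset and a shared rifle bullet impact concurrency set.
# A modulator slightly varies pitch over a short fixed delay, which feeds a quality-level switch.
# The first two quality slots randomize across ten waves; the last two use a five-wave subset.
# Both random nodes are excluded from branch culling.
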SoundCue /VampireStakeGameplay/Items/StakeLauncher/Audio/SC_StakeLauncher_Impact_Hard
root=Modulator_0
nodes [
 {id=Delay_0, children [QualityLevel_0], properties {delayMax=0.15, delayMin=0.15}}
 {id=Modulator_0, children [Delay_0], properties {pitchMax=1.1, pitchMin=1}}
 {id=QualityLevel_0, children [Random_2,Random_2,Random_3,Random_3]}
 {
  id=Random_2
  children [
   WavePlayer_10
   WavePlayer_11
   WavePlayer_12
   WavePlayer_13
   WavePlayer_14
   WavePlayer_15
   WavePlayer_16
   WavePlayer_18
   WavePlayer_19
   WavePlayer_20
  ]
  properties {bShouldExcludeFromBranchCulling=true, weights [1,1,1,1,1,1,1,1,1,1]}
 }
 {
  id=Random_3
  children [WavePlayer_27, WavePlayer_26, WavePlayer_23, WavePlayer_22, WavePlayer_21]
  properties {bShouldExcludeFromBranchCulling=true, weights [1,1,1,1,1]}
 }
 {id=WavePlayer_10, properties {soundWaveAssetPtr=/VampireStakeGameplay/Items/StakeLauncher/Audio/SoundWaves/StakeLauncher_Impact_Hard_01}}
 {id=WavePlayer_11, properties {soundWaveAssetPtr=/VampireStakeGameplay/Items/StakeLauncher/Audio/SoundWaves/StakeLauncher_Impact_Hard_02}}
 {id=WavePlayer_12, properties {soundWaveAssetPtr=/VampireStakeGameplay/Items/StakeLauncher/Audio/SoundWaves/StakeLauncher_Impact_Hard_03}}
 {id=WavePlayer_13, properties {soundWaveAssetPtr=/VampireStakeGameplay/Items/StakeLauncher/Audio/SoundWaves/StakeLauncher_Impact_Hard_04}}
 {id=WavePlayer_14, properties {soundWaveAssetPtr=/VampireStakeGameplay/Items/StakeLauncher/Audio/SoundWaves/StakeLauncher_Impact_Hard_05}}
 {id=WavePlayer_15, properties {soundWaveAssetPtr=/VampireStakeGameplay/Items/StakeLauncher/Audio/SoundWaves/StakeLauncher_Impact_Hard_06}}
 {id=WavePlayer_16, properties {soundWaveAssetPtr=/VampireStakeGameplay/Items/StakeLauncher/Audio/SoundWaves/StakeLauncher_Impact_Hard_07}}
 {id=WavePlayer_18, properties {soundWaveAssetPtr=/VampireStakeGameplay/Items/StakeLauncher/Audio/SoundWaves/StakeLauncher_Impact_Hard_08}}
 {id=WavePlayer_19, properties {soundWaveAssetPtr=/VampireStakeGameplay/Items/StakeLauncher/Audio/SoundWaves/StakeLauncher_Impact_Hard_09}}
 {id=WavePlayer_20, properties {soundWaveAssetPtr=/VampireStakeGameplay/Items/StakeLauncher/Audio/SoundWaves/StakeLauncher_Impact_Hard_10}}
 {id=WavePlayer_21, properties {soundWaveAssetPtr=&StakeLauncher_Impact_Hard_10}}
 {id=WavePlayer_22, properties {soundWaveAssetPtr=&StakeLauncher_Impact_Hard_08}}
 {id=WavePlayer_23, properties {soundWaveAssetPtr=&StakeLauncher_Impact_Hard_06}}
 {id=WavePlayer_26, properties {soundWaveAssetPtr=&StakeLauncher_Impact_Hard_04}}
 {id=WavePlayer_27, properties {soundWaveAssetPtr=&StakeLauncher_Impact_Hard_03}}
]
properties {
 attenuationSettings=/VampireStakeGameplay/Items/StakeLauncher/Audio/Settings/Fort_Stake_Impact_Att
 maxDistance=5800
 soundClassObject=&Fort_Impacts
 volumeMultiplier=3
 attenuationOverrides {reverbDistanceMax=3000, spatializationAlgorithm=SPATIALIZATION_HRTF}
 concurrencyOverrides {maxCount=4, resolutionRule=PreventNew}
 concurrencySet [&RifleBulletImpact_Conc]
}
