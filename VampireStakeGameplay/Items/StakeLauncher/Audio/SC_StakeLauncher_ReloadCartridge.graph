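# Sound cue for the Stake Launcher's cartridge reload sound.
# A Random node picks one of four equally weighted reload wave variations and feeds it through a Modulator at the root.
# The cue plays quieter and higher-pitched, using the shared weapon reload attenuation and the weapon foley sound class.
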SoundCue /VampireStakeGameplay/Items/StakeLauncher/Audio/SC_StakeLauncher_ReloadCartridge
root=Modulator_0
nodes [
 {id=Modulator_0, children [Random_0]}
 {id=Random_0, children [WavePlayer_0,WavePlayer_1,WavePlayer_2,WavePlayer_7], properties {weights [1,1,1,1]}}
 {id=WavePlayer_0, properties {soundWaveAssetPtr=/VampireStakeGameplay/Items/StakeLauncher/Audio/SoundWaves/StakeLauncher_ReloadCartridge_04}}
 {id=WavePlayer_1, properties {soundWaveAssetPtr=/VampireStakeGameplay/Items/StakeLauncher/Audio/SoundWaves/StakeLauncher_ReloadCartridge_03}}
 {id=WavePlayer_2, properties {soundWaveAssetPtr=/VampireStakeGameplay/Items/StakeLauncher/Audio/SoundWaves/StakeLauncher_ReloadCartridge_02}}
 {id=WavePlayer_7, properties {soundWaveAssetPtr=/VampireStakeGameplay/Items/StakeLauncher/Audio/SoundWaves/StakeLauncher_ReloadCartridge_01}}
]
properties {
 attenuationSettings=&Fort_Weapon_Reload_Att
 maxDistance=2250
 pitchMultiplier=1.4
 soundClassObject=&Fort_WeaponFoley
 volumeMultiplier=0.4
}
