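# Sound cue for the Stake Launcher's full reload-end sound in the VampireStakeGameplay plugin.
# A Random node picks one of three equally weighted wave variations, then feeds it through a Modulator at the root.
# The cue uses the shared weapon reload attenuation and weapon foley sound class, with its volume reduced.
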SoundCue /VampireStakeGameplay/Items/StakeLauncher/Audio/SC_StakeLauncher_ReloadEnd_Full
root=Modulator_0
nodes [
 {id=Modulator_0, children [Random_1]}
 {id=Random_1, children [WavePlayer_8, WavePlayer_9, WavePlayer_10], properties {weights [1,1,1]}}
 {id=WavePlayer_10, properties {soundWaveAssetPtr=/VampireStakeGameplay/Items/StakeLauncher/Audio/SoundWaves/StakeLauncher_ReloadEnd_Full_03}}
 {id=WavePlayer_8, properties {soundWaveAssetPtr=/VampireStakeGameplay/Items/StakeLauncher/Audio/SoundWaves/StakeLauncher_ReloadEnd_Full_01}}
 {id=WavePlayer_9, properties {soundWaveAssetPtr=/VampireStakeGameplay/Items/StakeLauncher/Audio/SoundWaves/StakeLauncher_ReloadEnd_Full_02}}
]
properties {
 attenuationSettings=&Fort_Weapon_Reload_Att
 maxDistance=2250
 soundClassObject=&Fort_WeaponFoley
 volumeMultiplier=0.5
}
