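# Sound cue for the start of the Stake Launcher's reload.
# A Random node picks one of three reload-start wave variations with equal weight and feeds a Modulator at the root.
# The cue plays as weapon foley with shared reload attenuation, at reduced volume and raised pitch.
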SoundCue /VampireStakeGameplay/Items/StakeLauncher/Audio/SC_StakeLauncher_ReloadStart
root=Modulator_0
nodes [
 {id=Modulator_0, children [Random_0]}
 {id=Random_0, children [WavePlayer_0,WavePlayer_1,WavePlayer_2], properties {weights [1,1,1]}}
 {id=WavePlayer_0, properties {soundWaveAssetPtr=/VampireStakeGameplay/Items/StakeLauncher/Audio/SoundWaves/StakeLauncher_ReloadStart_03}}
 {id=WavePlayer_1, properties {soundWaveAssetPtr=/VampireStakeGameplay/Items/StakeLauncher/Audio/SoundWaves/StakeLauncher_ReloadStart_02}}
 {id=WavePlayer_2, properties {soundWaveAssetPtr=/VampireStakeGameplay/Items/StakeLauncher/Audio/SoundWaves/StakeLauncher_ReloadStart_01}}
]
properties {
 attenuationSettings=&Fort_Weapon_Reload_Att
 maxDistance=2250
 pitchMultiplier=1.2
 soundClassObject=&Fort_WeaponFoley
 volumeMultiplier=0.3
}
