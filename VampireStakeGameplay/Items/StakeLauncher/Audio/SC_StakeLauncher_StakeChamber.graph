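# Sound cue for the Stake Launcher weapon's stake-chambering foley.
# A Random node picks one of four equally weighted StakeChamber wave variants and feeds it through a Modulator at the root to add variation.
# Playback uses the weapon reload attenuation and weapon foley sound class at reduced volume.
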SoundCue /VampireStakeGameplay/Items/StakeLauncher/Audio/SC_StakeLauncher_StakeChamber
root=Modulator_0
nodes [
 {id=Modulator_0, children [Random_0]}
 {
  id=Random_0
  children [WavePlayer_1,WavePlayer_2,WavePlayer_3,WavePlayer_4]
  properties {bShouldExcludeFromBranchCulling=true, weights [1,1,1,1]}
 }
 {id=WavePlayer_1, properties {soundWaveAssetPtr=/VampireStakeGameplay/Items/StakeLauncher/Audio/SoundWaves/StakeLauncher_StakeChamber_01}}
 {id=WavePlayer_2, properties {soundWaveAssetPtr=/VampireStakeGameplay/Items/StakeLauncher/Audio/SoundWaves/StakeLauncher_StakeChamber_02}}
 {id=WavePlayer_3, properties {soundWaveAssetPtr=/VampireStakeGameplay/Items/StakeLauncher/Audio/SoundWaves/StakeLauncher_StakeChamber_03}}
 {id=WavePlayer_4, properties {soundWaveAssetPtr=/VampireStakeGameplay/Items/StakeLauncher/Audio/SoundWaves/StakeLauncher_StakeChamber_04}}
]
properties {
 attenuationSettings=&Fort_Weapon_Reload_Att
 maxDistance=2250
 soundClassObject=&Fort_WeaponFoley
 volumeMultiplier=0.451
}
