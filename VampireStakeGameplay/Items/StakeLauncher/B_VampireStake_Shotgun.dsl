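; Stake launcher shotgun weapon that leaves stake impact FX embedded in hit surfaces and breaks them when that actor is destroyed.
; Impact handling orients the FX toward the weapon and checks surfaces and pawn attachments to suppress impaling visuals.
; A map links FX to damageable building actors; their death-effects delegate, avoiding life-span delay, replays FX with stakes hidden, while deactivation clears tracking.
; Targeting also drives an ADS audio parameter.
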(event FortWeapon|EventOnPlayImpactFX
  (params (HitResult :type Hit%20Result%20Structure%20%28by%20ref%29) (ImpactPhysicalSurface :type EPhysicalSurface%20Enum :default "SurfaceType_Default") (SpawnedPSC :type FXSystem%20Component%20Object%20Reference))
  (|CallParentFunction HitResult ImpactPhysicalSurface SpawnedPSC)
  (bind (_particlesystemcomponent _hitactor _btrackactor) (CallFunction|HandleHit HitResult ImpactPhysicalSurface SpawnedPSC))
  ;[0.6262 1 0.332 0.56] Make the stake break if whatever it's embedded in gets destroyed
  (if _btrackactor
    (Utilities|Map|Add (Variables|Default|GetSpawnedPSC_Map) _particlesystemcomponent _hitactor) ; Track the stakes that are embedded in something
    (Components|Activation|BindEventtoOnComponentDeactivated _particlesystemcomponent (EventDispatchers|CreateEvent :!event Custom|OnComponentDeactivated_Event)) ; Bind to the stake FX ending
    (EventDispatchers|AssignOnDeathPlayEffectsDelegate _hitactor (EventDispatchers|CreateEvent :!event Custom|OnDeathPlayEffectsDelegate_Event)) ; Bind to the actor playing its death FX. This gets around Life Span delaying OnDestroyed().
  ))

(event Custom|OnComponentDeactivated_Event
  (params (Component :type Actor%20Component%20Object%20Reference))
  ;[0.2611 0.5021 1 0.5] Event bindings
  (CallFunction|HandleDeactivatedFX Component))

(event Custom|OnDeathPlayEffectsDelegate_Event
  (params (Damage :type Float%20%28single-precision%29) (DamageTags :type Gameplay%20Tag%20Container%20Structure%20%28by%20ref%29) (Momentum :type Vector) (HitInfo :type Hit%20Result%20Structure%20%28by%20ref%29) (InstigatedBy :type Fort%20Pawn%20Object%20Reference) (DamageCauser :type Actor%20Object%20Reference) (EffectContext :type Gameplay%20Effect%20Context%20Handle%20Structure))
  (bind (_bblockinghit _binitialoverlap _time _distance _location _impactpoint _normal _impactnormal _physmat _hitactor) (Collision|BreakHitResult HitInfo))
  ;[0.2611 0.5021 1 0.5] Event bindings
  (CallFunction|HandleDestroyedActor _hitactor))

(event EventOnSetTargeting
  (params (bNewIsTargeting :type Boolean))
  (bind _returnvalue (AudioParameter|GetAudioParameterComponent))
  (|CallParentFunction bNewIsTargeting)
  (Utilities|IsValid _returnvalue
    (:Is%20Valid
      (Audio|Parameter|SetBooleanParameter _returnvalue "IsADS" bNewIsTargeting :!class /Script/AudioExtensions.AudioParameterControllerInterface))))

(fn HandleDeactivatedFX
  (params (FX_Component :type Object%20Reference))
  (bind _spawnedpsc_map (Variables|Default|GetSpawnedPSC_Map))
  (bind _asfxsystem_component (Utilities|Casting|CastToFXSystemComponent FX_Component)) ; Also inherently checks validity
  (if (.ReturnValue (Utilities|Map|Find _spawnedpsc_map _asfxsystem_component))
    (LocalVariables|SetAssociatedActor (Utilities|Map|Find _spawnedpsc_map _asfxsystem_component)))
  (CallFunction|RemoveandUnbindFXSystem _asfxsystem_component) ; Remove refs and bindings w/ this FX
  ;[0.5963 0.496 1 0.5] Was the deactivated FX the only stake in the associated actor?
  (bind _associatedactor (LocalVariables|GetAssociatedActor)
    (:then
      (bind _values (Utilities|Map|Values (Variables|Default|GetSpawnedPSC_Map)))
      (if (not (Utilities|Array|ContainsItem _values _associatedactor))
        (CallFunction|RemoveandUnbindActor _associatedactor) ; Remove refs and bindings w/ this actor
      ))
    (:else)))

(fn HandleDestroyedActor
  (params (DestroyedActor :type Actor%20Object%20Reference))
  (bind _asbuilding_actor (Utilities|Casting|CastToBuildingActor DestroyedActor)) ; Also inherently checks validity
  ;[0.4475 0.5786 1 0.5] Make sure there's at least one stake embedded in the actor
  (bind _values (Utilities|Map|Values (Variables|Default|GetSpawnedPSC_Map)))
  (LocalVariables|SetCurrentIndex (Utilities|Array|FindItem _values _asbuilding_actor))
  ;[0.7341 0.6706 1 0.5] While at least one stake is embedded in the actor...
  (while (>= (LocalVariables|GetCurrentIndex) 0)
    ;[0.8583 0.8726 1 0.5] Get the first FX system (stake) associated with this actor
    (bind _keys (Utilities|Map|Keys (Variables|Default|GetSpawnedPSC_Map)))
    (bind _output_get (LocalVariables|SetAssociatedFX (Utilities|Array|Get(acopy) _keys (LocalVariables|GetCurrentIndex)))) ; This will always be valid because if the FX\nended, its entry would've been removed by\nHandleDeactivatedFX().
    (Components|Activation|Activate _output_get true) ; Reset the FX for the explosion/flash
    (Effects|Components|ParticleSystem|SetBoolParameter _output_get "ShowStake" false) ; Hide the stake
    (CallFunction|RemoveandUnbindFXSystem _output_get) ; Remove refs and bindings w/ this FX
    ;[1 0.602 0.8864 0.5] Check again for the While loop
    (bind _values_2 (Utilities|Map|Values (Variables|Default|GetSpawnedPSC_Map)))
    (LocalVariables|SetCurrentIndex (Utilities|Array|FindItem _values_2 _asbuilding_actor)))
  (CallFunction|RemoveandUnbindActor _asbuilding_actor))

(fn HandleHit
  (params (HitResult :type Hit%20Result%20Structure%20%28by%20ref%29) (ImpactPhysicalSurface :type EPhysicalSurface%20Enum :default "SurfaceType_Default") (ParticleSystemComp :type FXSystem%20Component%20Object%20Reference))
  (bind _hitresult (Variables|GetHitResult))
  (bind (_bblockinghit _binitialoverlap _time _distance _location _impactpoint _normal _impactnormal _physmat _hitactor) (Collision|BreakHitResult _hitresult))
  (Utilities|FlowControl|Sequence
    (:then_0
      (Utilities|IsValid ParticleSystemComp
        (:Is%20Valid
          ;[0.1399 0 1] Event OnPlayImpactFX
          ; Set the rotation to match the direction of the "projectile" instead of using the surface normal, set the impact VFX when it hits a pawn so that there is no visual stake/impalement
          (if (Utilities|Enum|Equal(Enum) (Variables|GetImpactPhysicalSurface) "SurfaceType8")
            ;[0.4107 0.5581 1] Did we hit something we want to break on?
            (Utilities|FlowControl|Sequence :!label _label
              (:then_0
                (switch Utilities|FlowControl|Switch|SwitchonEPhysicalSurface (Variables|GetImpactPhysicalSurface)
                  (:SurfaceType4
                    _)
                  (:SurfaceType5
                    _)
                  (:SurfaceType6
                    _)
                  (:SurfaceType7
                    _)
                  (:SurfaceType17
                    _)
                  (:SurfaceType19
                    _)
                  (:SurfaceType20
                    _)
                  (:SurfaceType21
                    _)
                  (:SurfaceType22
                    _)))
              (:then_1
                ;;[0.3797 1 0.4925 0.5] Catch when we've hit something attached to a pawn (that is therefore likely to move/disappear) which doesn't use one of the selected physical surfaces
                (for _array_element (Actor|GetAttachedActors _hitactor true false)
                  (if (Utilities|IsA _array_element (LocalVariables|GetPawnClass))
                    (Effects|Components|ParticleSystem|SetBoolParameter (Variables|GetParticleSystemComp) "HitPlayer" true :!label _label_2) ; Prevent the stake mesh from spawning
                    (return 0 0 false)))
                (bind _result (CallFunction|IsActorAttachedtoPawn _hitactor))
                (if (not _result)
                  (bind _asbuilding_actor (Utilities|Casting|CastToBuildingActor _hitactor)) ; Also inherently checks validity
                  (return (Variables|GetParticleSystemComp) _asbuilding_actor (Variables|Actor|GetbCanBeDamaged _asbuilding_actor)))))
            (else
              (Transformation|SetRelativeRotation (Variables|GetParticleSystemComp) (Math|Rotator|FindLookatRotation _impactpoint (Transformation|GetWorldLocation (Variables|Visual|GetWeaponMesh))) false false) ; Rotate FX in direction of weapon
              (if (not (Utilities|IsA _hitactor (LocalVariables|GetPawnClass)))
                (goto _label))))
          (goto _label_2))))
    (:then_1
      (return 0 0 false))))

(fn IsActorAttachedToPawn
  (params (Actor :type Actor%20Object%20Reference))
  (LocalVariables|SetCurrentActor Actor)
  (for _ (range 1 (+ (LocalVariables|GetMaxDepth) 1))
    (bind _currentactor (LocalVariables|GetCurrentActor)
      (:then
        (bind _output_get (LocalVariables|SetCurrentActor (Actor|GetAttachParentActor _currentactor)))
        (Utilities|IsValid _output_get
          (:Is%20Valid
            (if (Utilities|IsA _output_get "/Script/Engine.Pawn")
              (return true)))
          (:Is%20Not%20Valid
            (return false))))
      (:else
        (return false)))))

(fn RemoveAndUnbind_Actor
  (params (BuildingActor :type Building%20Actor%20Object%20Reference))
  ;[1 0.3334 0.2917 0.5] Stop tracking this Building Actor
  (EventDispatchers|UnbindEventfromOnDeathPlayEffectsDelegate BuildingActor (EventDispatchers|CreateEvent :!event OnDeathPlayEffectsDelegate_Event)))

(fn RemoveAndUnbind_FXSystem
  (params (FXSystemComp :type FXSystem%20Component%20Object%20Reference))
  ;[1 0.3334 0.2917 0.5] Stop tracking this FXSystem Component
  (Utilities|Map|Remove (Variables|Default|GetSpawnedPSC_Map) FXSystemComp)
  (Components|Activation|UnbindEventfromOnComponentDeactivated FXSystemComp (EventDispatchers|CreateEvent :!event OnComponentDeactivated_Event)))

(fn ConstructionScript
  (|CallParentFunction))
