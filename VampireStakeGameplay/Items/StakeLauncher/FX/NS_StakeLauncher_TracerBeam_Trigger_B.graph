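# Niagara system for the stake launcher's tracer beam (Trigger B variant), drawing mesh tracers along shots between supplied start and end positions.
# One looping emitter uses shared weapon modules to spawn when firing, initialize tracers from User position arrays and orient them along fire direction.
# Travel fitting, bullet-size updates and curve-driven dynamic material parameters animate three mesh renderers, passing beam start and length to materials.
# Distance and LOD remaps adjust lifetime and scaling.
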NiagaraSystem /VampireStakeGameplay/Items/StakeLauncher/FX/NS_StakeLauncher_TracerBeam_Trigger_B
dependencies {
 usedDataInterfaces [&NiagaraDataInterfaceCurve]
 usedDynamicInputs []
 usedModules [&SystemState, /Game/Effects/Niagara/Modules/Weapons/FN_Setup_Weapons]
}
emitters {Tracers_InfLoop {
  emitterSpawnScript [/SetVariables(Emitter.DebugTimeMultiplier=1) {isSetParameters=true}]
  emitterUpdateScript [
   &EmitterState()
   /Game/Effects/Niagara/Modules/Weapons/FN_Spawn_When_Firing(Age=0, MaxRange:NiagaraFloat=$User.MaxRange)
  ]
  particleSpawnScript [
   /Game/Effects/Niagara/Modules/Weapons/FN_Initialize_Tracers(
    BeamEndPositions:NiagaraDataInterfaceArrayPosition=$User.BeamEndPositions
    BeamStartPositions:NiagaraDataInterfaceArrayPosition=$User.BeamStartPositions
    Remap By Distance Alpha=DataInterface() {propertyValues {Curve {keys [
        {arriveTangent=0.0002, interpMode=RCIM_Cubic, leaveTangent=0.0002, tangentMode=RCTM_User, time=0, value=0}
        {arriveTangent=0.0001, arriveTangentWeight=833.3334, interpMode=RCIM_Cubic, leaveTangent=0.0001, tangentMode=RCTM_User, time=2500, value=0.35}
        <at 15000 0.8>
        {arriveTangentWeight=6666.667, interpMode=RCIM_Cubic, tangentMode=RCTM_User, time=35000, value=1}
       ]}}}
    Remap By LODDistance=DataInterface() {propertyValues {Curve {keys [<at 500 0>,<at 2500 1>]}}}
   )
   &InitializeParticle(
    Color=<rgba 1 0.3543 0.026 0.9>
    Lifetime=&Multiply_Float(
     A=&Multiply_Float(
      A=&ClampFloat(Float=$Particles.RemapByDistance, Min=0.5)
      B=&RandomRangeFloat(Minimum=0.7, Maximum=0.8)
     )
     B=$Emitter.DebugTimeMultiplier
    )
    Mesh Scale Mode=::NewEnumerator3
    Position=$Particles.BeamStartPosition
    Position Mode=::NewEnumerator1
    UsePositionOffset=false
    Write Material Random=true
   )
   &DynamicMaterialParameters(
    Index 0 Param 4=&Multiply_Float(A=$Particles.BeamTile, B=&Lerp_Float(A=0.8, B=1.35, Alpha=$Particles.RemapByLODDistance))
    Index 1 Param 1=&RandomRangeFloat(Minimum=0.85, Maximum=1.05)
    Index 1 Param 2=&RandomRangeFloat(Minimum=0, Maximum=1)
    Module.Param0WriteEnabled=false
    Module.Param0WriteEnabled1=false
    Module.Param1WriteEnabled=false
    Module.Param1WriteEnabled1=true
    Module.Param2WriteEnabled=false
    Module.Param2WriteEnabled1=false
    Module.Param3WriteEnabled=true
    Module.Param3WriteEnabled1=false
    Write Parameter Index 1=true
   ) {module=DynamicMaterialParameters001}
   /Niagara/Modules/Update/Orientation/V2/OrientMeshToVector(Look At Direction:Vector3f=$Particles.FireDirection, Override Quat=false, UseOptionalSideAxis=false, UseOptionalUpAxis=false)
   /SetVariables(Particles.BulletDynParams:Vector4f=<1 1 1 1>) {isSetParameters=true}
   /SetVariables(Particles.BulletColor:LinearColor=<rgba 1 1 1 0.3>) {isSetParameters=true}
  ]
  particleUpdateScript [
   &ParticleState()
   /Niagara/Modules/Update/Utility/FrameCounter()
   /Game/Effects/Niagara/Modules/Weapons/FN_Fit_Tracer_Travel(DurationMax=0.5, DurationMin=0.1, MaxFrames=7)
   &DynamicMaterialParameters(
    Index 0 Param 1=&FloatFromCurve(FloatCurve=DataInterface() {propertyValues {Curve {keys [
         {arriveTangent=-0.4135, interpMode=RCIM_Cubic, leaveTangent=-0.4135, tangentMode=RCTM_User, time=0, value=1}
         {arriveTangent=-1.3177, interpMode=RCIM_Cubic, leaveTangent=-1.3177, tangentMode=RCTM_User, time=1, value=0}
        ]}}})
    Index 0 Param 2=&FloatFromCurve(
     FloatCurve=DataInterface() {propertyValues {Curve {keys [
         {arriveTangent=5.2627, interpMode=RCIM_Cubic, leaveTangent=5.2627, tangentMode=RCTM_User, time=0, value=0}
         {arriveTangent=0.8931, interpMode=RCIM_Cubic, leaveTangent=0.8931, tangentMode=RCTM_User, time=0.2287, value=0.7}
         {arriveTangentWeight=0.2571, interpMode=RCIM_Cubic, tangentMode=RCTM_User, time=1, value=1}
        ]}}}
     Scale Curve=&Multiply_Float(
      A=&Lerp_Float(A=0.5, B=1, Alpha=$Particles.RemapByDistance)
      B=&RandomRangeFloat(Minimum=-10, Maximum=-8)
     )
    )
    Index 0 Param 3=&FloatFromCurve(
     FloatCurve=DataInterface() {propertyValues {Curve {keys [
         {arriveTangent=-0.9882, time=0.2, value=1}
         {arriveTangent=-1.2626, interpMode=RCIM_Cubic, leaveTangent=-1.2626, tangentMode=RCTM_User, time=0.5, value=0.6}
         {arriveTangent=-10.4271, interpMode=RCIM_Cubic, leaveTangent=-10.4271, tangentMode=RCTM_User, time=0.6, value=0}
        ]}}}
     CurveIndex=$Particles.FrameLimitedAge
    )
    Index 0 Param 4=&Multiply_Float(
     A=&MakeFloatFromVector(Vector=$Particles.Scale, Channel=::NewEnumerator0)
     B=&Lerp_Float(A=0.8, B=1.35, Alpha=$Particles.RemapByLODDistance)
    )
    Index 1 Param 1=&FloatFromCurve(
     FloatCurve=DataInterface() {propertyValues {Curve {keys [
         {arriveTangent=0.0005, interpMode=RCIM_Cubic, leaveTangent=0.0005, tangentMode=RCTM_User, time=0, value=0.7}
         {arriveTangent=0.9056, interpMode=RCIM_Cubic, leaveTangent=0.9056, tangentMode=RCTM_User, time=0.5, value=0.87}
         {arriveTangent=0.8994, interpMode=RCIM_Cubic, leaveTangent=0.8994, tangentMode=RCTM_User, time=0.5714, value=1.214}
         {arriveTangent=0.8128, interpMode=RCIM_Cubic, leaveTangent=0.8128, tangentMode=RCTM_User, time=1, value=1.5}
        ]}}}
     Scale Curve=&Lerp_Float(A=0.5, B=1, Alpha=$StackContext.RemapByLODDistance)
    )
    Index 3 Param 1=&MakeFloatFromVector(
     Vector=&ConvertPositionToVector(Input Position=$Particles.BeamStartPosition, Conversion Operation:ENiagara_LWCConvertPosToVec=::NewEnumerator0)
     Channel=::NewEnumerator0
    )
    Index 3 Param 2=&MakeFloatFromVector(
     Vector=&ConvertPositionToVector(Input Position=$Particles.BeamStartPosition, Conversion Operation:ENiagara_LWCConvertPosToVec=::NewEnumerator0)
     Channel=::NewEnumerator1
    )
    Index 3 Param 3=&MakeFloatFromVector(
     Vector=&ConvertPositionToVector(Input Position=$Particles.BeamStartPosition, Conversion Operation:ENiagara_LWCConvertPosToVec=::NewEnumerator0)
     Channel=::NewEnumerator2
    )
    Index 3 Param 4=&DistanceBetweenPositions(A=$Particles.BeamStartPosition, B=$Particles.BeamEndPosition)
    Module.Param0WriteEnabled=true
    Module.Param0WriteEnabled1=true
    Module.Param0WriteEnabled3=true
    Module.Param1WriteEnabled=true
    Module.Param1WriteEnabled1=false
    Module.Param1WriteEnabled3=true
    Module.Param2WriteEnabled=true
    Module.Param2WriteEnabled1=false
    Module.Param2WriteEnabled3=true
    Module.Param3WriteEnabled=false
    Module.Param3WriteEnabled1=false
    Module.Param3WriteEnabled3=true
    Write Parameter Index 1=true
    Write Parameter Index 3=true
   )
   /Game/Effects/Niagara/Modules/Weapons/FN_Update_Bullet_Travel(
    Bullet Size Factor=DataInterface() {propertyValues {Curve {keys [<at 0 0>,<at 0.05 1>]}}}
    Bullet Size Far:Vector3f=<4 4 4>
    Bullet Size Near:Vector3f=<2 2 2>
    Tracer Mesh Size Alpha=DataInterface() {propertyValues {Curve {keys [
        {interpMode=RCIM_Cubic, tangentMode=RCTM_Break, time=0.65, value=0}
        {arriveTangent=9.0989, arriveTangentWeight=0.1167, interpMode=RCIM_Cubic, tangentMode=RCTM_Break, time=1, value=1}
       ]}}}
   )
   /SetVariables(Particles.DebugScale:Vector3f=<5 5 5>) {enabled=false, isSetParameters=true}
  ]
  rendererClasses [&NiagaraMeshRendererProperties]
  renderers [
   {rendererClass=&NiagaraMeshRendererProperties, rendererIndex=0}
   {rendererClass=&NiagaraMeshRendererProperties, rendererIndex=1}
   {rendererClass=&NiagaraMeshRendererProperties, rendererIndex=2}
  ]
 }}
summary {
 systemName=NS_StakeLauncher_TracerBeam_Trigger_B
 emitters [{emitterName=Tracers_InfLoop, rendererClasses [&NiagaraMeshRendererProperties]}]
 userVariables [
  User.BeamEndPositions {
   type=NiagaraDataInterfaceArrayPosition
   description=""
   defaultValue {
    struct=NiagaraExt_VariableValue_DataInterface
    value {
     dataInterface=&self:NiagaraDataInterfaceArrayPosition_1
     dataInterfaceClass=&NiagaraDataInterfaceArrayPosition
    }
   }
  }
  User.BeamStartPositions {
   type=NiagaraDataInterfaceArrayPosition
   description=""
   defaultValue {
    struct=NiagaraExt_VariableValue_DataInterface
    value {
     dataInterface=&self:NiagaraDataInterfaceArrayPosition_0
     dataInterfaceClass=&NiagaraDataInterfaceArrayPosition
    }
   }
  }
  User.MaxRange {type=NiagaraFloat, defaultValue=27000, description=""}
  User.TriggerInt {type=NiagaraInt32}
 ]
}
